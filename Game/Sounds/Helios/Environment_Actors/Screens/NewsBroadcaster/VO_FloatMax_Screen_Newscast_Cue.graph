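# Sound cue for looping newscast voice-over on an in-world screen, spatialized with a TV screen attenuation asset.
# A loop waits a randomized delay, then randomly picks one of four equally weighted newscast pairs.
# Each pair concatenates two voice lines with a short fixed pause between them, so chatter varies between repeats.
# A concurrency override caps simultaneous instances, and playback volume is reduced.
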SoundCue /Game/Sounds/Helios/Environment_Actors/Screens/NewsBroadcaster/VO_FloatMax_Screen_Newscast_Cue
root=Looping_0
nodes [
 {id=Concatenator_0, children [WavePlayer_1,Delay_1], properties {inputVolume [1,1]}}
 {id=Concatenator_1, children [WavePlayer_3,Delay_2], properties {inputVolume [1,1]}}
 {id=Concatenator_2, children [WavePlayer_6,Delay_3], properties {inputVolume [1,1]}}
 {id=Concatenator_3, children [WavePlayer_4,Delay_4], properties {inputVolume [1,1]}}
 {id=Delay_0, children [Random_1], properties {delayMax=10, delayMin=0.4}}
 {id=Delay_1, children [WavePlayer_0], properties {delayMax=0.2, delayMin=0.2}}
 {id=Delay_2, children [WavePlayer_2], properties {delayMax=0.4, delayMin=0.4}}
 {id=Delay_3, children [WavePlayer_5], properties {delayMax=0.4, delayMin=0.4}}
 {id=Delay_4, children [WavePlayer_11], properties {delayMax=0.4, delayMin=0.4}}
 {id=Looping_0, children [Delay_0]}
 {id=Random_1, children [Concatenator_0, Concatenator_1, Concatenator_2, Concatenator_3], properties {weights [1,1,1,1]}}
 {id=WavePlayer_0, properties {soundWaveAssetPtr=&/VO_S31_Newscast_010}}
 {id=WavePlayer_1, properties {soundWaveAssetPtr=&/VO_S31_Newscast_020}}
 {id=WavePlayer_11, properties {soundWaveAssetPtr=&/VO_S31_Newscast_050}}
 {id=WavePlayer_2, properties {soundWaveAssetPtr=&/VO_S31_Newscast_030}}
 {id=WavePlayer_3, properties {soundWaveAssetPtr=&/VO_S31_Newscast_070}}
 {id=WavePlayer_4, properties {soundWaveAssetPtr=&/VO_S31_Newscast_090}}
 {id=WavePlayer_5, properties {soundWaveAssetPtr=&/VO_S31_Newscast_100}}
 {id=WavePlayer_6, properties {soundWaveAssetPtr=&/VO_S31_Newscast_110}}
]
properties {
 attenuationSettings=/Game/Sounds/Helios/Environment_Actors/Screens/Spot_TV_Screen_VO_Sphere_Att
 bOverrideConcurrency=true
 maxDistance=801
 soundClassObject=&Fort_Gadgets_Items
 volumeMultiplier=0.45
 concurrencyOverrides {maxCount=2}
}
